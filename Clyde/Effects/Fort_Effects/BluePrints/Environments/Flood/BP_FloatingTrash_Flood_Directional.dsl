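; Directional floating-trash surface Blueprint for flood environments, driven by a material on its static mesh.
; The construction script makes a dynamic instance of the directional trash material in slot 0 and sets WaterBodyIndex on it from a variable.
; It then creates a second instance from the same source in that slot before setting OverallDirection on the mesh's materials.
; BeginPlay, overlap and Tick events are empty.
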(event EventBeginPlay)

(event Collision|EventActorBeginOverlap
  (params (OtherActor :type Actor%20Object%20Reference)))

(event EventTick
  (params (DeltaSeconds :type Float%20%28single-precision%29)))

(fn ConstructionScript
  ;[1 1 1] Set scalar parameter values in the trash shader
  (bind _returnvalue (Rendering|Material|CreateDynamicMaterialInstance (Variables|Default|GetStaticMesh) 0 "/Clyde/Athena/Effects/Materials/Environment/Flood/MI_Flood_Trash_Surface_Directional_Inst.MI_Flood_Trash_Surface_Directional_Inst" "None" :!class /Script/Engine.PrimitiveComponent))
  (Rendering|Material|SetScalarParameterValue _returnvalue "WaterBodyIndex" (Variables|Variables|GetWaterBodyIndex) :!class /Script/Engine.MaterialInstanceDynamic)
  (Rendering|Material|CreateDynamicMaterialInstance (Variables|Default|GetStaticMesh) 0 "/Clyde/Athena/Effects/Materials/Environment/Flood/MI_Flood_Trash_Surface_Directional_Inst.MI_Flood_Trash_Surface_Directional_Inst" "None" :!class /Script/Engine.PrimitiveComponent)
  (Rendering|Material|SetVectorParameterValueonMaterials (Variables|Default|GetStaticMesh) "OverallDirection" (Variables|Variables|GetOverallDirection)))
