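; Blueprint for rotating floating trash in flood environments, with motion apparently driven by the surface material.
; Construction Script builds dynamic instances of the rotational trash material on the static mesh, passing water body index, forward and reverse rotation speeds from variables.
; Each parameter gets a separate instance created on slot 0 from the same parent, so later creations replace earlier ones.
; BeginPlay, overlap and Tick events are empty.
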(event EventBeginPlay)

(event Collision|EventActorBeginOverlap
  (params (OtherActor :type Actor%20Object%20Reference)))

(event EventTick
  (params (DeltaSeconds :type Float%20%28single-precision%29)))

(fn ConstructionScript
  ;[1 1 1] Set scalar parameter values in the trash shader
  (bind _returnvalue (Rendering|Material|CreateDynamicMaterialInstance (Variables|Default|GetStaticMesh) 0 "/Clyde/Athena/Effects/Materials/Environment/Flood/MI_Flood_Trash_Surface_Rotational_Inst.MI_Flood_Trash_Surface_Rotational_Inst" "None" :!class /Script/Engine.PrimitiveComponent))
  (Rendering|Material|SetScalarParameterValue _returnvalue "WaterBodyIndex" (Variables|Variables|GetWaterBodyIndex) :!class /Script/Engine.MaterialInstanceDynamic)
  (bind _returnvalue_2 (Rendering|Material|CreateDynamicMaterialInstance (Variables|Default|GetStaticMesh) 0 "/Clyde/Athena/Effects/Materials/Environment/Flood/MI_Flood_Trash_Surface_Rotational_Inst.MI_Flood_Trash_Surface_Rotational_Inst" "None" :!class /Script/Engine.PrimitiveComponent))
  (Rendering|Material|SetScalarParameterValue _returnvalue_2 "RotationSpeed" (Variables|Variables|GetRotationSpeed) :!class /Script/Engine.MaterialInstanceDynamic)
  (bind _returnvalue_3 (Rendering|Material|CreateDynamicMaterialInstance (Variables|Default|GetStaticMesh) 0 "/Clyde/Athena/Effects/Materials/Environment/Flood/MI_Flood_Trash_Surface_Rotational_Inst.MI_Flood_Trash_Surface_Rotational_Inst" "None" :!class /Script/Engine.PrimitiveComponent))
  (Rendering|Material|SetScalarParameterValue _returnvalue_3 "RotationSpeedReverse" (Variables|Variables|GetRotationSpeedReverse) :!class /Script/Engine.MaterialInstanceDynamic))
